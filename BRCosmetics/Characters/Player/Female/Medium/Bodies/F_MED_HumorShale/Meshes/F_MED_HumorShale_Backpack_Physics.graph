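# Physics asset for the HumorShale female medium character's backpack mesh, set up for a dangling necklace and ribbon.
# Capsules on spine, clavicle and pec bones form the upper-torso shapes, joined by free-rotation constraints.
# The necklace is a three-bone chain and the ribbon a root with four two-bone branches, each hung from a deform bone with tight angular limits.
# Extra cross-constraints tie ribbon branches together and the ribbon root back to spine_05.
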PhysicsAsset /BRCosmetics/Characters/Player/Female/Medium/Bodies/F_MED_HumorShale/Meshes/F_MED_HumorShale_Backpack_Physics
bodies {
 clavicle_l [{shapeType=Capsule, length=11.1474, radius=5.6925, shapeName=clavicle_l_capsule, center=<3.1884 -0.532 -1.3729>, rotation=<rot 90 180 0>}]
 clavicle_r [{shapeType=Capsule, length=11.1474, radius=5.6925, shapeName=clavicle_r_capsule, center=<-3.1884 0.532 1.3729>, rotation=<rot 90 0 0>}]
 deform_necklace [{shapeType=Capsule, length=1, radius=1, shapeName=deform_necklace_capsule}]
 deform_pec_aim_l [{
   shapeType=Capsule
   length=8.1798
   radius=5.4985
   shapeName=deform_pec_aim_l_capsule
   center=<5.1828 -0.3322 -2.8622>
   rotation=<rot 90 177.7662 0>
  }]
 deform_pec_aim_r [{
   shapeType=Capsule
   length=8.1798
   radius=5.4985
   shapeName=deform_pec_aim_r_capsule
   center=<-5.1828 0.3322 2.8622>
   rotation=<rot 90 -2.2338 0>
  }]
 deform_ribbon [{shapeType=Capsule, length=1, radius=1, shapeName=deform_ribbon_capsule}]
 dyn_necklace [{shapeType=Capsule, length=6.0995, radius=3.4806, shapeName=dyn_necklace_capsule, center=<1.9813 -2.5646 0>}]
 dyn_necklace_1 [{shapeType=Capsule, length=7.9475, radius=3.4806, shapeName=dyn_necklace_1_capsule, center=<1.9813 -2.6783 0>}]
 dyn_necklace_2 [{shapeType=Capsule, length=5.5183, radius=2.6979, shapeName=dyn_necklace_2_capsule, center=<3.4686 -2.01 0>, rotation=<rot 90 8.5478 0>}]
 dyn_ribbon [{shapeType=Capsule, length=8.1097, radius=3.2318, shapeName=dyn_ribbon_capsule, center=<4.5367 -2.461 0>, rotation=<rot 90 0 0>}]
 dyn_ribbon_1_down_l [{shapeType=Capsule, length=4, radius=2, shapeName=dyn_ribbon_1_down_l_capsule, center=<2.4151 -1.1067 0>, rotation=<rot 90 0 0>}]
 dyn_ribbon_1_down_r [{shapeType=Capsule, length=4, radius=2, shapeName=dyn_ribbon_1_down_r_capsule, center=<2.4151 -1.1067 0>, rotation=<rot 90 0 0>}]
 dyn_ribbon_1_up_l [{shapeType=Capsule, length=4, radius=2, shapeName=dyn_ribbon_1_up_l_capsule, center=<2.4151 -1.1067 0>}]
 dyn_ribbon_1_up_r [{shapeType=Capsule, length=4, radius=2, shapeName=dyn_ribbon_1_up_r_capsule, center=<2.4151 -1.1067 0>}]
 dyn_ribbon_down_l [{shapeType=Capsule, length=4, radius=2, shapeName=dyn_ribbon_down_l_capsule, center=<2.4151 -1.1067 0>, rotation=<rot 90 0 0>}]
 dyn_ribbon_down_r [{shapeType=Capsule, length=4, radius=2, shapeName=dyn_ribbon_down_r_capsule, center=<2.4151 -1.1067 0>, rotation=<rot 90 0 0>}]
 dyn_ribbon_up_l [{shapeType=Capsule, length=4, radius=2, shapeName=dyn_ribbon_up_l_capsule, center=<2.4151 -2.0489 0>}]
 dyn_ribbon_up_r [{shapeType=Capsule, length=4.2692, radius=2.9132, shapeName=dyn_ribbon_up_r_capsule, center=<2.8978 -2.0481 -1.758>, rotation=<rot 90 0 0>}]
 spine_04 [{shapeType=Capsule, length=7.5157, radius=10.2343, shapeName=spine_04_capsule, center=<0 3.6397 0>}]
 spine_05 [
  {shapeType=Capsule, length=11.5098, radius=5.3094, shapeName=spine_05_capsule, center=<6.3757 -1.6419 0>}
  {shapeType=Capsule, length=13.1347, radius=8.0838, shapeName=spine_05_capsule}
  {shapeType=Capsule, length=10.5205, radius=19.4795, shapeName=spine_05_capsule, center=<0 -32.6442 0>, rotation=<rot 90 0 0>}
 ]
}
constraints [
 {
  bone1Name=spine_05
  bone2Name=spine_04
  swing1LimitDegrees=45
  swing1Motion=Free
  swing2LimitDegrees=45
  swing2Motion=Free
  twistLimitDegrees=45
  twistMotion=Free
 }
 {
  bone1Name=clavicle_l
  bone2Name=spine_05
  swing1LimitDegrees=45
  swing1Motion=Free
  swing2LimitDegrees=45
  swing2Motion=Free
  twistLimitDegrees=45
  twistMotion=Free
 }
 {
  bone1Name=clavicle_r
  bone2Name=spine_05
  swing1LimitDegrees=45
  swing1Motion=Free
  swing2LimitDegrees=45
  swing2Motion=Free
  twistLimitDegrees=45
  twistMotion=Free
 }
 {
  bone1Name=deform_necklace
  bone2Name=spine_05
  swing1LimitDegrees=45
  swing1Motion=Free
  swing2LimitDegrees=45
  swing2Motion=Free
  twistLimitDegrees=45
  twistMotion=Free
 }
 {bone1Name=dyn_necklace, bone2Name=deform_necklace, swing1LimitDegrees=25, swing2LimitDegrees=15, twistLimitDegrees=1}
 {bone1Name=dyn_necklace_1, bone2Name=dyn_necklace, swing1LimitDegrees=25, swing2LimitDegrees=15, twistLimitDegrees=5}
 {bone1Name=dyn_necklace_2, bone2Name=dyn_necklace_1, swing1LimitDegrees=25, swing2LimitDegrees=15, twistLimitDegrees=5}
 {
  bone1Name=deform_ribbon
  bone2Name=spine_05
  swing1LimitDegrees=45
  swing1Motion=Free
  swing2LimitDegrees=45
  swing2Motion=Free
  twistLimitDegrees=45
  twistMotion=Free
 }
 {bone1Name=dyn_ribbon, bone2Name=deform_ribbon, swing1LimitDegrees=35, swing2LimitDegrees=25, twistLimitDegrees=2}
 {bone1Name=dyn_ribbon_down_l, bone2Name=dyn_ribbon, swing1LimitDegrees=5, swing2LimitDegrees=5, twistLimitDegrees=2}
 {bone1Name=dyn_ribbon_1_down_l, bone2Name=dyn_ribbon_down_l, swing1LimitDegrees=5, swing2LimitDegrees=5, twistLimitDegrees=2}
 {bone1Name=dyn_ribbon_down_r, bone2Name=dyn_ribbon, swing1LimitDegrees=5, swing2LimitDegrees=5, twistLimitDegrees=2}
 {bone1Name=dyn_ribbon_1_down_r, bone2Name=dyn_ribbon_down_r, swing1LimitDegrees=5, swing2LimitDegrees=5, twistLimitDegrees=2}
 {bone1Name=dyn_ribbon_up_l, bone2Name=dyn_ribbon, swing1LimitDegrees=5, swing2LimitDegrees=5, twistLimitDegrees=2}
 {bone1Name=dyn_ribbon_1_up_l, bone2Name=dyn_ribbon_up_l, swing1LimitDegrees=5, swing2LimitDegrees=5, twistLimitDegrees=2}
 {bone1Name=dyn_ribbon_up_r, bone2Name=dyn_ribbon, swing1LimitDegrees=5, swing2LimitDegrees=5, twistLimitDegrees=2}
 {bone1Name=dyn_ribbon_1_up_r, bone2Name=dyn_ribbon_up_r, swing1LimitDegrees=5, swing2LimitDegrees=5, twistLimitDegrees=2}
 {bone1Name=dyn_ribbon_up_r, bone2Name=dyn_ribbon_down_r, swing1LimitDegrees=45, swing2LimitDegrees=45, twistLimitDegrees=5}
 {bone1Name=dyn_ribbon_up_l, bone2Name=dyn_ribbon_down_l, swing1LimitDegrees=45, swing2LimitDegrees=45, twistLimitDegrees=5}
 {bone1Name=dyn_ribbon_up_r, bone2Name=dyn_ribbon_up_l, swing1LimitDegrees=45, swing2LimitDegrees=45, twistLimitDegrees=45}
 {bone1Name=dyn_ribbon_down_r, bone2Name=dyn_ribbon_down_l, swing1LimitDegrees=45, swing2LimitDegrees=45, twistLimitDegrees=45}
 {
  bone1Name=dyn_ribbon_1_up_r
  bone2Name=dyn_ribbon_1_down_r
  swing1LimitDegrees=45
  swing2LimitDegrees=45
  swing2Motion=Free
  twistLimitDegrees=45
  twistMotion=Free
 }
 {
  bone1Name=deform_pec_aim_l
  bone2Name=spine_04
  swing1LimitDegrees=45
  swing1Motion=Free
  swing2LimitDegrees=45
  swing2Motion=Free
  twistLimitDegrees=45
  twistMotion=Free
 }
 {
  bone1Name=deform_pec_aim_r
  bone2Name=spine_04
  swing1LimitDegrees=45
  swing1Motion=Free
  swing2LimitDegrees=45
  swing2Motion=Free
  twistLimitDegrees=45
  twistMotion=Free
 }
 {bone1Name=spine_05, bone2Name=dyn_ribbon, swing1LimitDegrees=1, swing2LimitDegrees=1, twistLimitDegrees=1}
 {bone1Name=dyn_ribbon_1_up_r, bone2Name=dyn_ribbon_1_up_l, swing1LimitDegrees=45, swing2LimitDegrees=45, twistLimitDegrees=1}
]
