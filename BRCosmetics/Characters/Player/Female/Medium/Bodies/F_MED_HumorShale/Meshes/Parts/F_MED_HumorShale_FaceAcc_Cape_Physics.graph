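# Physics asset for the HumorShale female character's FaceAcc Cape mesh part, covering hair and earring bones but no cape-named bones.
# Capsules on upper spine, neck, head, clavicles and upper arms form the body shapes.
# Side hair, front hair and earring chains hang from head-linked deform bones, with tight angular limits on hair and free rotation on earrings.
# Extra free-rotation constraints tie the fourth side-hair segments to the clavicles and each other.
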PhysicsAsset /BRCosmetics/Characters/Player/Female/Medium/Bodies/F_MED_HumorShale/Meshes/Parts/F_MED_HumorShale_FaceAcc_Cape_Physics
bodies {
 clavicle_l [{
   shapeType=Capsule
   length=7
   radius=5.25
   shapeName=clavicle_l_capsule
   center=<5.2922 -1.0853 -0.5563>
   rotation=<rot 78.9134 -0.0527 -0.0544>
  }]
 clavicle_r [{
   shapeType=Capsule
   length=7.75
   radius=5.5
   shapeName=clavicle_r_capsule
   center=<-5.3625 1.0182 0.4147>
   rotation=<rot 84.6697 -0.1121 179.8867>
  }]
 deform_earring_l [{shapeType=Capsule, length=0.505, radius=0.505, shapeName=deform_earring_l_capsule, rotation=<rot 0 -90 0>}]
 deform_earring_r [{shapeType=Capsule, length=0.505, radius=0.505, shapeName=deform_earring_r_capsule, rotation=<rot 0 -90 0>}]
 deform_hair_f_l [{shapeType=Sphere, radius=2, shapeName=deform_hair_f_l_sphere}]
 deform_hair_f_r [{shapeType=Sphere, radius=2, shapeName=deform_hair_f_r_sphere}]
 deform_hair_l [{shapeType=Sphere, radius=2, shapeName=deform_hair_l_sphere}]
 deform_hair_r [{shapeType=Sphere, radius=2, shapeName=deform_hair_r_sphere}]
 dyn_earring_l [{shapeType=Capsule, length=1.8686, radius=2.2345, shapeName=dyn_earring_l_capsule, center=<1.9213 1.3091 0>, rotation=<rot 90 0 0>}]
 dyn_earring_r [{shapeType=Capsule, length=1.8686, radius=2.2345, shapeName=dyn_earring_r_capsule, center=<1.9213 1.3091 0>, rotation=<rot 90 0 0>}]
 dyn_hair_1_l [{shapeType=Sphere, radius=3, shapeName=dyn_hair_1_l_sphere, center=<1.6413 0 0>}]
 dyn_hair_1_r [{shapeType=Sphere, radius=3, shapeName=dyn_hair_1_r_sphere, center=<1.6413 0 0>}]
 dyn_hair_2_l [{shapeType=Capsule, length=6.6263, radius=4.25, shapeName=dyn_hair_2_l_capsule, center=<3.8797 0 0>, rotation=<rot 90 0 0>}]
 dyn_hair_2_r [{shapeType=Capsule, length=6.6263, radius=4.25, shapeName=dyn_hair_2_r_capsule, center=<3.8797 0 0>, rotation=<rot 90 0 0>}]
 dyn_hair_3_l [{shapeType=Capsule, length=8.453, radius=4.25, shapeName=dyn_hair_3_l_capsule, center=<4.0132 0 0>, rotation=<rot 90 0 0>}]
 dyn_hair_3_r [{shapeType=Capsule, length=8.453, radius=4.25, shapeName=dyn_hair_3_r_capsule, center=<4.0132 0 0>, rotation=<rot 90 0 0>}]
 dyn_hair_4_l [{shapeType=Capsule, length=7.1953, radius=4.5, shapeName=dyn_hair_4_l_capsule, center=<3.8117 0 0>, rotation=<rot 90 0 0>}]
 dyn_hair_4_r [{shapeType=Capsule, length=6.3405, radius=4.832, shapeName=dyn_hair_4_r_capsule, center=<3.8117 0 0>, rotation=<rot 90 0 0>}]
 dyn_hair_5_l [{shapeType=Capsule, length=5.9481, radius=2.9629, shapeName=dyn_hair_5_l_capsule, center=<2.4583 0 0>, rotation=<rot 90 0 0>}]
 dyn_hair_5_r [{shapeType=Capsule, length=4.6915, radius=3.0623, shapeName=dyn_hair_5_r_capsule, center=<1.6644 0 0>, rotation=<rot 90 0 0>}]
 dyn_hair_f_1_l [{shapeType=Capsule, length=6.4522, radius=1.5478, shapeName=dyn_hair_f_1_l_capsule, center=<3.0623 -0.6121 0>, rotation=<rot 90 0 0>}]
 dyn_hair_f_1_r [{shapeType=Capsule, length=6.4522, radius=1.5478, shapeName=dyn_hair_f_1_r_capsule, center=<3.0623 -0.6121 0>, rotation=<rot 90 0 0>}]
 dyn_hair_f_2_l [{shapeType=Capsule, length=6.4522, radius=1.5478, shapeName=dyn_hair_f_2_l_capsule, center=<3.0623 0 0>, rotation=<rot 90 0 0>}]
 dyn_hair_f_2_r [{shapeType=Capsule, length=6.4522, radius=1.5478, shapeName=dyn_hair_f_2_r_capsule, center=<3.0623 0 0>, rotation=<rot 90 0 0>}]
 dyn_hair_f_l [{shapeType=Capsule, length=5.4424, radius=2.5576, shapeName=dyn_hair_f_l_capsule, center=<3.0623 0 0>, rotation=<rot 90 0 0>}]
 dyn_hair_f_r [{shapeType=Capsule, length=5.4424, radius=2.5576, shapeName=dyn_hair_f_r_capsule, center=<3.0623 0 0>, rotation=<rot 90 0 0>}]
 head [
  {shapeType=Capsule, length=4.4145, radius=6.8355, shapeName=head_capsule, center=<4.5655 0 0>, rotation=<rot 90 -55.2374 0>}
  {shapeType=Capsule, length=3.9578, radius=6.25, shapeName=head_capsule, center=<3.4613 -2.0995 0>}
  {shapeType=Capsule, length=4.9293, radius=5.0707, shapeName=head_capsule, center=<-0.0591 3.8239 0>, rotation=<rot 90 -33.7041 0>}
 ]
 neck_01 [{shapeType=Capsule, length=9.1745, radius=4, shapeName=neck_01_capsule, center=<4.6347 0 0>, rotation=<rot 90 0 0>}]
 spine_04 [{shapeType=Capsule, length=10.75, radius=9.5, shapeName=spine_04_capsule, center=<4.7296 1.4253 0>}]
 spine_05 [
  {shapeType=Capsule, length=13.75, radius=7.25, shapeName=spine_05_capsule, center=<3.7654 -0.5421 0>}
  {shapeType=Capsule, length=10.8729, radius=11.9031, shapeName=spine_05_capsule, center=<-0.8137 -5.3064 0>}
 ]
 upperarm_l [{shapeType=Capsule, length=20.25, radius=4.25, shapeName=upperarm_l_capsule, center=<9.9048 0 -0.3109>, rotation=<rot 90 0 0>}]
 upperarm_r [{shapeType=Capsule, length=19.3028, radius=5.1972, shapeName=upperarm_r_capsule, center=<-9.9048 0 0.3109>, rotation=<rot 90 180 0>}]
}
constraints [
 {
  bone1Name=spine_05
  bone2Name=spine_04
  swing1LimitDegrees=45
  swing1Motion=Free
  swing2LimitDegrees=45
  swing2Motion=Free
  twistLimitDegrees=45
  twistMotion=Free
 }
 {
  bone1Name=neck_01
  bone2Name=spine_05
  swing1LimitDegrees=45
  swing1Motion=Free
  swing2LimitDegrees=45
  swing2Motion=Free
  twistLimitDegrees=45
  twistMotion=Free
 }
 {
  bone1Name=head
  bone2Name=neck_01
  swing1LimitDegrees=45
  swing1Motion=Free
  swing2LimitDegrees=45
  swing2Motion=Free
  twistLimitDegrees=45
  twistMotion=Free
 }
 {
  bone1Name=deform_earring_l
  bone2Name=head
  swing1LimitDegrees=45
  swing1Motion=Free
  swing2LimitDegrees=45
  swing2Motion=Free
  twistLimitDegrees=45
  twistMotion=Free
 }
 {
  bone1Name=dyn_earring_l
  bone2Name=deform_earring_l
  swing1LimitDegrees=45
  swing1Motion=Free
  swing2LimitDegrees=45
  swing2Motion=Free
  twistLimitDegrees=45
  twistMotion=Free
 }
 {
  bone1Name=deform_earring_r
  bone2Name=head
  swing1LimitDegrees=45
  swing1Motion=Free
  swing2LimitDegrees=45
  swing2Motion=Free
  twistLimitDegrees=45
  twistMotion=Free
 }
 {
  bone1Name=dyn_earring_r
  bone2Name=deform_earring_r
  swing1LimitDegrees=45
  swing1Motion=Free
  swing2LimitDegrees=45
  swing2Motion=Free
  twistLimitDegrees=45
  twistMotion=Free
 }
 {
  bone1Name=deform_hair_l
  bone2Name=head
  swing1LimitDegrees=45
  swing1Motion=Free
  swing2LimitDegrees=45
  swing2Motion=Free
  twistLimitDegrees=45
  twistMotion=Free
 }
 {bone1Name=dyn_hair_1_l, bone2Name=deform_hair_l, swing1LimitDegrees=15, swing2LimitDegrees=15, twistLimitDegrees=5}
 {bone1Name=dyn_hair_2_l, bone2Name=dyn_hair_1_l, swing1LimitDegrees=35, swing2LimitDegrees=35, twistLimitDegrees=5}
 {bone1Name=dyn_hair_3_l, bone2Name=dyn_hair_2_l, swing1LimitDegrees=35, swing2LimitDegrees=35, twistLimitDegrees=5}
 {bone1Name=dyn_hair_4_l, bone2Name=dyn_hair_3_l, swing1LimitDegrees=35, swing2LimitDegrees=15, twistLimitDegrees=5}
 {bone1Name=dyn_hair_5_l, bone2Name=dyn_hair_4_l, swing1LimitDegrees=35, swing2LimitDegrees=15, twistLimitDegrees=5}
 {
  bone1Name=deform_hair_r
  bone2Name=head
  swing1LimitDegrees=45
  swing1Motion=Free
  swing2LimitDegrees=45
  swing2Motion=Free
  twistLimitDegrees=45
  twistMotion=Free
 }
 {bone1Name=dyn_hair_1_r, bone2Name=deform_hair_r, swing1LimitDegrees=15, swing2LimitDegrees=15, twistLimitDegrees=5}
 {bone1Name=dyn_hair_2_r, bone2Name=dyn_hair_1_r, swing1LimitDegrees=35, swing2LimitDegrees=35, twistLimitDegrees=5}
 {bone1Name=dyn_hair_3_r, bone2Name=dyn_hair_2_r, swing1LimitDegrees=35, swing2LimitDegrees=35, twistLimitDegrees=5}
 {bone1Name=dyn_hair_4_r, bone2Name=dyn_hair_3_r, swing1LimitDegrees=35, swing2LimitDegrees=15, twistLimitDegrees=5}
 {bone1Name=dyn_hair_5_r, bone2Name=dyn_hair_4_r, swing1LimitDegrees=35, swing2LimitDegrees=15, twistLimitDegrees=5}
 {
  bone1Name=clavicle_l
  bone2Name=spine_05
  swing1LimitDegrees=45
  swing1Motion=Free
  swing2LimitDegrees=45
  swing2Motion=Free
  twistLimitDegrees=45
  twistMotion=Free
 }
 {
  bone1Name=clavicle_r
  bone2Name=spine_05
  swing1LimitDegrees=45
  swing1Motion=Free
  swing2LimitDegrees=45
  swing2Motion=Free
  twistLimitDegrees=45
  twistMotion=Free
 }
 {
  bone1Name=upperarm_l
  bone2Name=clavicle_l
  swing1LimitDegrees=45
  swing1Motion=Free
  swing2LimitDegrees=45
  swing2Motion=Free
  twistLimitDegrees=45
  twistMotion=Free
 }
 {
  bone1Name=upperarm_r
  bone2Name=clavicle_r
  swing1LimitDegrees=45
  swing1Motion=Free
  swing2LimitDegrees=45
  swing2Motion=Free
  twistLimitDegrees=45
  twistMotion=Free
 }
 {
  bone1Name=spine_04
  bone2Name=spine_03
  swing1LimitDegrees=45
  swing1Motion=Free
  swing2LimitDegrees=45
  swing2Motion=Free
  twistLimitDegrees=45
  twistMotion=Free
 }
 {
  bone1Name=dyn_hair_4_l
  bone2Name=clavicle_l
  swing1LimitDegrees=45
  swing1Motion=Free
  swing2LimitDegrees=45
  swing2Motion=Free
  twistLimitDegrees=45
  twistMotion=Free
 }
 {
  bone1Name=dyn_hair_4_r
  bone2Name=clavicle_r
  swing1LimitDegrees=45
  swing1Motion=Free
  swing2LimitDegrees=45
  swing2Motion=Free
  twistLimitDegrees=45
  twistMotion=Free
 }
 {
  bone1Name=dyn_hair_4_r
  bone2Name=dyn_hair_4_l
  swing1LimitDegrees=45
  swing1Motion=Free
  swing2LimitDegrees=45
  swing2Motion=Free
  twistLimitDegrees=45
  twistMotion=Free
 }
 {
  bone1Name=deform_hair_f_l
  bone2Name=head
  swing1LimitDegrees=45
  swing1Motion=Free
  swing2LimitDegrees=45
  swing2Motion=Free
  twistLimitDegrees=45
  twistMotion=Free
 }
 {bone1Name=dyn_hair_f_l, bone2Name=deform_hair_f_l, swing1LimitDegrees=5, swing2LimitDegrees=5, twistLimitDegrees=0, twistMotion=Locked}
 {bone1Name=dyn_hair_f_1_l, bone2Name=dyn_hair_f_l, swing1LimitDegrees=15, swing2LimitDegrees=30, twistLimitDegrees=5}
 {bone1Name=dyn_hair_f_2_l, bone2Name=dyn_hair_f_1_l, swing1LimitDegrees=20, swing2LimitDegrees=30, twistLimitDegrees=10}
 {
  bone1Name=deform_hair_f_r
  bone2Name=head
  swing1LimitDegrees=45
  swing1Motion=Free
  swing2LimitDegrees=45
  swing2Motion=Free
  twistLimitDegrees=45
  twistMotion=Free
 }
 {bone1Name=dyn_hair_f_r, bone2Name=deform_hair_f_r, swing1LimitDegrees=5, swing2LimitDegrees=5, twistLimitDegrees=0, twistMotion=Locked}
 {bone1Name=dyn_hair_f_1_r, bone2Name=dyn_hair_f_r, swing1LimitDegrees=15, swing2LimitDegrees=30, twistLimitDegrees=5}
 {bone1Name=dyn_hair_f_2_r, bone2Name=dyn_hair_f_1_r, swing1LimitDegrees=20, swing2LimitDegrees=30, twistLimitDegrees=10}
]
